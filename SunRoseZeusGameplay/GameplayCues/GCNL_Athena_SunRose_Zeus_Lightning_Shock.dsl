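; Looping gameplay cue for the SunRose Zeus lightning shock, which puts a temporary material override on the affected player pawn.
; On loop start it waits briefly, then applies a soft-referenced material locally at a fixed priority and stores the handle.
; The override flags depend on whether the pawn has the first-person camera gameplay tag.
; On removal it uses the stored handle to clear the override.
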(event EventOnLoopingStart(EitherVFX-preferthisversion)
  (params (MyTarget :type Actor%20Object%20Reference) (Parameters :type Gameplay%20Cue%20Parameters%20Structure%20%28by%20ref%29) (ParticleComponents :type Array%20of%20FXSystem%20Component%20Object%20References) (AudioComponents :type Array%20of%20Audio%20Component%20Object%20References))
  (Utilities|FlowControl|Delay 0 0.1)
  (bind _asfort_player_pawn (Utilities|Casting|CastToFortPlayerPawn MyTarget))
  (bind _returnvalue (Class|FortPlayerPawn|ApplyMaterialOverrideLocalwithFlags :self _asfort_player_pawn :Material (Utilities|ResolveSoftReference (Variables|Default|GetMaterial)) :Priority 111.0 :Flags (select (GameplayTags|HasMatchingGameplayTag _asfort_player_pawn "(TagName=\"Gameplay.Camera.FirstPerson\")") 5 7)))
  (Variables|Default|SetMaterial%20Override%20Handdle _returnvalue))

(event EventOnRemoval(EitherVFX-preferthisversion)
  (params (MyTarget :type Actor%20Object%20Reference) (Parameters :type Gameplay%20Cue%20Parameters%20Structure%20%28by%20ref%29) (ParticleComponents :type Array%20of%20FXSystem%20Component%20Object%20References) (AudioComponents :type Array%20of%20Audio%20Component%20Object%20References) (BurstCameraShakeInstance :type Camera%20Shake%20Base%20Object%20Reference) (BurstDecalInstance :type Decal%20Actor%20Object%20Reference))
  (bind _asfort_player_pawn (Utilities|Casting|CastToFortPlayerPawn MyTarget))
  (Class|FortPlayerPawn|RemoveMaterialOverrideLocal _asfort_player_pawn (Variables|Default|GetMaterial%20Override%20Handdle)))

(fn ConstructionScript)
