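; Burst gameplay cue for the SunRose Zeus lightning strike that plays the impact explosion effect and sound.
; OnBurst calls the parent, then stores the target as PlayerPawn, the cue location as HitLocation and the raw magnitude as BoltHeight.
; It spawns an auto-release pooled Niagara explosion and the strike sound at HitLocation, using PlayerPawn as world context.
; BoltHeight is set but not read in this graph.
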(fn OnBurst(EitherVFX-preferthisversion)
  (params (MyTarget :type Actor%20Object%20Reference) (Parameters :type Gameplay%20Cue%20Parameters%20Structure%20%28by%20ref%29) (ParticleComponents :type Array%20of%20FXSystem%20Component%20Object%20References) (AudioComponents :type Array%20of%20Audio%20Component%20Object%20References) (BurstCameraShakeInstance :type Camera%20Shake%20Base%20Object%20Reference) (BurstDecalInstance :type Decal%20Actor%20Object%20Reference))
  (bind (_normalizedmagnitude _rawmagnitude _effectcontext _matchedtagname _originaltag _aggregatedsourcetags _aggregatedtargettags _location) (Ability|GameplayCue|BreakGameplayCueParameters Parameters))
  (|CallParentFunction MyTarget Parameters ParticleComponents AudioComponents BurstCameraShakeInstance BurstDecalInstance)
  (LocalVariables|SetPlayerPawn MyTarget)
  (LocalVariables|SetHitLocation _location)
  (LocalVariables|SetBoltHeight _rawmagnitude)
  (Niagara|SpawnSystematLocation :WorldContextObject (LocalVariables|GetPlayerPawn) :SystemTemplate "/SunRoseZeusGameplay/Items/Lightning/FX/Explosion/NS_SunRose_Zeus_Explosion_F.NS_SunRose_Zeus_Explosion_F" :Location (LocalVariables|GetHitLocation) :Scale "1.000000,1.000000,1.000000" :bAutoDestroy true :bAutoActivate true :PoolingMethod "AutoRelease" :bPreCullCheck false)
  (Audio|SpawnSoundatLocation (LocalVariables|GetPlayerPawn) "/SunRoseZeusGameplay/Sounds/Ability/SunRose_ZeusLightning_Strike_Cue.SunRose_ZeusLightning_Strike_Cue" (LocalVariables|GetHitLocation))
  (return))
