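; Looping gameplay cue for the SunRose Zeus lightning telegraph, which limits how long each telegraph effect stays active.
; On looping start it calls the parent, then appends the last particle component received to the ActiveParticles array.
; After a delay read from the TelegraphDuration scalable float, EndTelegraph deactivates and removes the first array entry, so effects end oldest first.
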(event EventOnLoopingStart(EitherVFX-preferthisversion)
  (params (MyTarget :type Actor%20Object%20Reference) (Parameters :type Gameplay%20Cue%20Parameters%20Structure%20%28by%20ref%29) (ParticleComponents :type Array%20of%20FXSystem%20Component%20Object%20References) (AudioComponents :type Array%20of%20Audio%20Component%20Object%20References))
  (|CallParentFunction MyTarget Parameters ParticleComponents AudioComponents)
  (if (not (Utilities|Array|IsEmpty ParticleComponents))
    (Utilities|Array|Add (Variables|Default|GetActiveParticles) (Utilities|Array|Get(acopy) ParticleComponents (- (Utilities|Array|Length ParticleComponents) 1)))
    (Utilities|FlowControl|Delay 0 (ScalableFloat|GetValueatLevel (Variables|Default|GetRow_TelegraphDuration)))
    (CallFunction|EndTelegraph)))

(fn EndTelegraph
  (bind _activeparticles (Variables|Default|GetActiveParticles))
  (if (not (Utilities|Array|IsEmpty _activeparticles))
    (Components|Activation|Deactivate (Utilities|Array|Get(acopy) _activeparticles))
    (Utilities|Array|RemoveIndex _activeparticles)))

(fn ConstructionScript)
