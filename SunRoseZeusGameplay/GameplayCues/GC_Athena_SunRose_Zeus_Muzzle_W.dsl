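; Gameplay cue for the SunRose Zeus muzzle effect, spawning a one-shot Niagara system when the cue executes.
; OnExecute casts the target to PlayerPawn_Athena and spawns the configured system at the pawn's location, raised by a configurable Z offset.
; The effect faces the pawn's default targeting rotation and is placed in world space rather than attached to a muzzle socket.
; It then returns the parent OnExecute result.
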(fn OnExecute
  (params (MyTarget :type Actor%20Object%20Reference) (Parameters :type Gameplay%20Cue%20Parameters%20Structure%20%28by%20ref%29))
  (bind _asplayer_pawn_athena (Utilities|Casting|CastToPlayerPawn_Athena MyTarget))
  (bind (_returnvalue_x _returnvalue_y _returnvalue_z) (Math|Vector|BreakVector (Transformation|GetActorLocation _asplayer_pawn_athena)))
  (Niagara|SpawnSystematLocation 0 (Variables|Default|GetMuzzle%20FX%20Niagara) (Math|Vector|MakeVector _returnvalue_x _returnvalue_y (+ _returnvalue_z (Variables|Default|GetFX%20Z%20Ofs))) (Targeting|GetDefaultTargetingRotation _asplayer_pawn_athena))
  (return (|CallParentFunction MyTarget Parameters)))
