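; Gameplay cue for the explosion burst of the SunRose Zeus light lightning projectile.
; After the parent burst, it stores the first spawned particle component and, unless a CVar disables it, can attach that Niagara effect to the projectile's target static mesh.
; It shows a gunshot sound indicator for non-locally-controlled pawn targets.
; Radial force and post-FX actors spawn only off mobile and PC performance mode, at higher detail levels.
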(event EventOnBurst(EitherVFX-preferthisversion)
  (params (MyTarget :type Actor%20Object%20Reference) (Parameters :type Gameplay%20Cue%20Parameters%20Structure%20%28by%20ref%29) (ParticleComponents :type Array%20of%20FXSystem%20Component%20Object%20References) (AudioComponents :type Array%20of%20Audio%20Component%20Object%20References) (BurstCameraShakeInstance :type Camera%20Shake%20Base%20Object%20Reference) (BurstDecalInstance :type Decal%20Actor%20Object%20Reference))
  (bind (_normalizedmagnitude _rawmagnitude _effectcontext _matchedtagname _originaltag _aggregatedsourcetags _aggregatedtargettags _location _normal _instigator _effectcauser) (Ability|GameplayCue|BreakGameplayCueParameters Parameters))
  (bind _output (Utilities|Array|Get(acopy) ParticleComponents))
  (bind _returnvalue (Transformation|GetWorldLocation _output))
  (bind _returnvalue_2 (Math|Transform|MakeTransform _returnvalue))
  (|CallParentFunction MyTarget Parameters ParticleComponents AudioComponents BurstCameraShakeInstance BurstDecalInstance)
  (Variables|Default|SetSpawnedFX (Utilities|Array|Get(acopy) ParticleComponents))
  (Utilities|FlowControl|Sequence
    (:then_0
      (if (not (== (Fortnite|GetNumericCVar "fx.SunRoseLightning.DisableExplosionFXAttachLogic") 1.0))
        (bind _asprj_athena_sun_rose_zeus_lightning_light (Utilities|Casting|CastToPrj_Athena_SunRose_Zeus_Lightning_Light MyTarget))
        (if (Variables|/SunRoseZeusGameplay/Items/Lightning/Blueprints/Prj_Athena_SunRose_Zeus_Lightning_Light.Prj_Athena_SunRose_Zeus_Lightning_Light_C|GetShouldExplosionFXAttach _asprj_athena_sun_rose_zeus_lightning_light)
          (bind _asstatic_mesh_component (Utilities|Casting|CastToStaticMeshComponent (Variables|/SunRoseZeusGameplay/Items/Lightning/Blueprints/Prj_Athena_SunRose_Zeus_Lightning_Light.Prj_Athena_SunRose_Zeus_Lightning_Light_C|GetExplosionFXAttachTarget _asprj_athena_sun_rose_zeus_lightning_light)))
          (bind _asniagara_particle_system_component (Utilities|Casting|CastToNiagaraComponent (Variables|Default|GetSpawnedFX)))
          (Effects|Components|ParticleSystem|SetBoolParameter _asniagara_particle_system_component "isAttach" true)
          (Niagara|SetNiagaraVariable(Object) _asniagara_particle_system_component "TargetSMObject" _asstatic_mesh_component))))
    (:then_1
      (bind _asfort_pawn (Utilities|Casting|CastToFortPawn MyTarget)
        (:then
          (if (not (Pawn|IsLocallyControlled _asfort_pawn))
            (Fortnite|ShowSoundIndicatorLocal :WorldContextObject _asfort_pawn :TrackedActor _asfort_pawn :Location (Transformation|GetActorLocation _asfort_pawn) :MaxAudibleDistance (ScalableFloat|GetValueatLevel (Variables|Default|GetRow_AVIndicatorDistance)) :IndicatorType "Gunshot" :Instigator _asfort_pawn :Tint "(R=0.000000,G=0.000000,B=0.000000,A=0.000000)" :AttenuationMaxRangeMultiplier 1.0 :AttenuationIntensityMultiplier 1.0)))
        (:CastFailed
          _))
      (if (Math|Boolean|ANDBoolean (not (Fortnite|IsMobilePlatform)) (not (Class|PlatformHelpersLibrary|IsPcPerfMode 0)) (> (Rendering|GetRenderingDetailMode) 1))
        (Game|SpawnActorfromClass 0 "/Game/Effects/Niagara/Blueprints/PhsyicsFields/FN_RadialForce.FN_RadialForce_C" _returnvalue_2 "Undefined" "MultiplyWithRoot" 0 3.0 0.6 3e+03 1.5 0.1 0)
        (bind _asfort_player_pawn_athena (Utilities|Casting|CastToFortPlayerPawnAthena _effectcauser))
        (Game|SpawnActorfromClass 0 "/SunRoseZeusGameplay/Items/Lightning/FX/PostFX/BP_Force_PostFX.BP_Force_PostFX_C" _returnvalue_2 "Undefined" "OverrideRootScale" 0 0.0 1.0 _asfort_player_pawn_athena 0)))))
